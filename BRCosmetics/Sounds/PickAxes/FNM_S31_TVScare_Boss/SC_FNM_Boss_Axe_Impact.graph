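# Sound cue for the impact of the FNM S31 TVScare Boss pickaxe cosmetic.
# A random node picks one of six equally weighted wave players, which reuse the DomeRoof pickaxe impact recordings rather than waves of its own.
# It uses the shared pickaxe impact attenuation, the cosmetic pickaxe impact property sheet and the Fort_Impacts sound class.
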SoundCue /BRCosmetics/Sounds/PickAxes/FNM_S31_TVScare_Boss/SC_FNM_Boss_Axe_Impact
root=Random_2
nodes [
 {id=Random_2, children [WavePlayer_0,WavePlayer_1,WavePlayer_2,WavePlayer_3,WavePlayer_4,WavePlayer_5], properties {weights [1,1,1,1,1,1]}}
 {id=WavePlayer_0, properties {soundWaveAssetPtr=/BRCosmetics/Sounds/PickAxes/DomeRoof/IP_Pickaxe_DomeRoof_Impact_06}}
 {id=WavePlayer_1, properties {soundWaveAssetPtr=/BRCosmetics/Sounds/PickAxes/DomeRoof/IP_Pickaxe_DomeRoof_Impact_05}}
 {id=WavePlayer_2, properties {soundWaveAssetPtr=/BRCosmetics/Sounds/PickAxes/DomeRoof/IP_Pickaxe_DomeRoof_Impact_04}}
 {id=WavePlayer_3, properties {soundWaveAssetPtr=/BRCosmetics/Sounds/PickAxes/DomeRoof/IP_Pickaxe_DomeRoof_Impact_03}}
 {id=WavePlayer_4, properties {soundWaveAssetPtr=/BRCosmetics/Sounds/PickAxes/DomeRoof/IP_Pickaxe_DomeRoof_Impact_02}}
 {id=WavePlayer_5, properties {soundWaveAssetPtr=/BRCosmetics/Sounds/PickAxes/DomeRoof/IP_Pickaxe_DomeRoof_Impact_01}}
]
properties {
 attenuationSettings=/Game/Sounds/Attenuation/Fort_Pickaxe_Impact_Att
 audioPropertiesSheet=/BRCosmetics/Sounds/Settings/PropertySheets/Pickaxe/APS_Cosmetic_Pickaxe_Impact
 maxDistance=4300
 soundClassObject=&Fort_Impacts
 concurrencySet [null]
}
