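# Sound cue for the swing of the FNM Boss axe pickaxe cosmetic from the S31 TVScare Boss set.
# A Random node picks one of six equally weighted swing recordings and passes it through a Modulator at the root for variation.
# Playback uses a close-range pickaxe attenuation, the cosmetic pickaxe swing audio property sheet and a melee weapon sound class.
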SoundCue /BRCosmetics/Sounds/PickAxes/FNM_S31_TVScare_Boss/SC_FNM_Boss_Axe_Swing
root=Modulator_0
nodes [
 {id=Modulator_0, children [Random_1]}
 {
  id=Random_1
  children [WavePlayer_6, WavePlayer_7, WavePlayer_8, WavePlayer_9, WavePlayer_10, WavePlayer_11]
  properties {weights [1,1,1,1,1,1]}
 }
 {id=WavePlayer_10, properties {soundWaveAssetPtr=&/FNM_Boss_Axe_Swing_05}}
 {id=WavePlayer_11, properties {soundWaveAssetPtr=&/FNM_Boss_Axe_Swing_06}}
 {id=WavePlayer_6, properties {soundWaveAssetPtr=&/FNM_Boss_Axe_Swing_01}}
 {id=WavePlayer_7, properties {soundWaveAssetPtr=&/FNM_Boss_Axe_Swing_02}}
 {id=WavePlayer_8, properties {soundWaveAssetPtr=&/FNM_Boss_Axe_Swing_03}}
 {id=WavePlayer_9, properties {soundWaveAssetPtr=&/FNM_Boss_Axe_Swing_04}}
]
properties {
 attenuationSettings=/Game/Sounds/Attenuation/Fort_Weapon_Pickaxe_Close_Att
 audioPropertiesSheet=/BRCosmetics/Sounds/Settings/PropertySheets/Pickaxe/APS_Cosmetic_Pickaxe_Swing
 maxDistance=1536
 soundClassObject=&Fort_WeaponMelee_PSM
}
